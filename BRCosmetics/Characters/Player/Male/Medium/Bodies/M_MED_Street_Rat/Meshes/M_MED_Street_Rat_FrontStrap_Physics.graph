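# Physics asset for the front strap of the Street Rat medium male player body, set up for simulating the dangling strap.
# Box bodies on the dyn_strap bones form a constrained chain from base through parent, 01 and 02.
# The final joint has tighter angle values and locks Swing1.
# Unconstrained capsules on spine_01 and thigh_l (two on the thigh) approximate the torso and left leg as collision shapes.
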PhysicsAsset /BRCosmetics/Characters/Player/Male/Medium/Bodies/M_MED_Street_Rat/Meshes/M_MED_Street_Rat_FrontStrap_Physics
bodies {
 dyn_strap_01 [{shapeType=Box, extentX=3.3158, extentY=5.6644, extentZ=1.8953, center=<1.6766 -0.0001 0>}]
 dyn_strap_02 [{shapeType=Box, extentX=6.3158, extentY=8.4144, extentZ=1.6453, center=<2.5646 0 0>, rotation=<rot -12.4852 0 0>}]
 dyn_strap_base [{shapeType=Box, extentX=0.75, extentY=4.75, extentZ=0.5}]
 dyn_strap_parent [{shapeType=Box, extentX=7.8158, extentY=5.4144, extentZ=1.3953, center=<3.3849 0 -0.83>, rotation=<rot -16.8768 0 0>}]
 spine_01 [{shapeType=Capsule, length=11.5, radius=9.5}]
 thigh_l [
  {shapeType=Capsule, length=21.875, radius=8.125, center=<-10.255 -0.3632 0.0057>, rotation=<rot 90 0 4.1969>}
  {shapeType=Capsule, length=10.75, radius=7.5, center=<-13.8069 1.7567 0.6799>, rotation=<rot 87.18 0 9.0352>}
 ]
}
constraints [
 {bone1Name=dyn_strap_parent, bone2Name=dyn_strap_base, swing1LimitDegrees=15, swing2LimitDegrees=30, twistLimitDegrees=3}
 {bone1Name=dyn_strap_01, bone2Name=dyn_strap_parent, swing1LimitDegrees=15, swing2LimitDegrees=30, twistLimitDegrees=3}
 {bone1Name=dyn_strap_02, bone2Name=dyn_strap_01, swing1LimitDegrees=8, swing1Motion=Locked, swing2LimitDegrees=20, twistLimitDegrees=2}
]
