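# Physics asset for the Street Rat medium male player body, simulating its dangling bag, pouch, shirt back and three tassels.
# Each accessory hangs from a base body through angle-limited constraints, with the tassels built as four-link box chains.
# The front-left tassel chain attaches to the bag base instead of its own base, which joins that bag base with free rotation.
# Unconstrained pelvis, spine and left thigh shapes provide body collision volumes.
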PhysicsAsset /BRCosmetics/Characters/Player/Male/Medium/Bodies/M_MED_Street_Rat/Meshes/M_MED_Street_Rat_Physics
bodies {
 dyn_bag_base [{shapeType=Box, extentX=1.01, extentY=1.01, extentZ=1.01, rotation=<rot 0 -90 0>}]
 dyn_bag_parent [{shapeType=Box, extentX=18.6879, extentY=18.0824, extentZ=6.9037, center=<8.501 0 0>, rotation=<rot 0.1444 0 0>}]
 dyn_pouch_top_r_base [{shapeType=Box, extentX=1.5, extentY=1.5, extentZ=1.5}]
 dyn_pouch_top_r_parent [{shapeType=Box, extentX=9.8117, extentY=6.6135, extentZ=4.5101, center=<3.9088 0.0001 0>}]
 dyn_shirt_back_base [{shapeType=Box, extentX=0.5, extentY=0.5, extentZ=0.5}]
 dyn_shirt_back_parent [{
   shapeType=Capsule
   length=5.0609
   radius=3.2252
   shapeName=dyn_shirt_back_parent_capsule
   center=<2.6249 0 -2>
   rotation=<rot 0 -176.635 -89.1873>
  }]
 dyn_tassel_frnt_l_01 [{shapeType=Box, extentX=2.76, extentY=3.51, extentZ=3.51, center=<1.5259 0 0>}]
 dyn_tassel_frnt_l_02 [{shapeType=Box, extentX=2.76, extentY=3.51, extentZ=3.51, center=<1.4944 0 0>}]
 dyn_tassel_frnt_l_03 [{shapeType=Box, extentX=5.26, extentY=3.51, extentZ=3.51, center=<2.5067 0 0>}]
 dyn_tassel_frnt_l_base [{shapeType=Box, extentX=0.75, extentY=0.75, extentZ=0.75}]
 dyn_tassel_frnt_l_parent [{shapeType=Box, extentX=5.51, extentY=3.51, extentZ=3.51, center=<2.4984 0 0>}]
 dyn_tassel_frnt_r_01 [{shapeType=Box, extentX=4.01, extentY=2.76, extentZ=2.26, center=<2.0132 0 0>}]
 dyn_tassel_frnt_r_02 [{shapeType=Box, extentX=4.01, extentY=2.76, extentZ=2.26, center=<2.0132 0 0>}]
 dyn_tassel_frnt_r_03 [{shapeType=Box, extentX=5.01, extentY=2.76, extentZ=2.26, center=<2.478 0 0>}]
 dyn_tassel_frnt_r_base [{shapeType=Box, extentX=1, extentY=1, extentZ=1}]
 dyn_tassel_frnt_r_parent [{shapeType=Box, extentX=4.01, extentY=1.51, extentZ=1.51, center=<1.9428 0 -0.3598>, rotation=<rot -11.0758 0 0>}]
 dyn_tassel_side_01 [{shapeType=Box, extentX=3.76, extentY=3.51, extentZ=3.51, center=<1.9043 0 0>}]
 dyn_tassel_side_02 [{shapeType=Box, extentX=3.51, extentY=3.51, extentZ=3.51, center=<1.6138 0 0>}]
 dyn_tassel_side_03 [{shapeType=Box, extentX=6.51, extentY=3.51, extentZ=3.51, center=<2.6731 0 -0.0192>}]
 dyn_tassel_side_base [{shapeType=Box, extentX=1, extentY=1, extentZ=1}]
 dyn_tassel_side_parent [{shapeType=Box, extentX=3.01, extentY=3.51, extentZ=3.51, center=<1.3177 0 0>}]
 pelvis [{shapeType=Box, extentX=12.5, extentY=7.5, extentZ=17.5, center=<0 -7.7129 0>, rotation=<rot 0 15 0>}]
 spine_01 [
  {shapeType=Capsule, length=12.625, radius=11.625, center=<-13.5929 -2.4681 2.5731>}
  {shapeType=Capsule, length=12.625, radius=11.625, center=<-3.5929 -2.4681 3.7397>}
 ]
 spine_04 [
  {shapeType=Box, extentX=21.25, extentY=17, extentZ=21.25, shapeName=Front_Right, center=<0.9977 3.9075 6.5083>, rotation=<rot 0 -8.0415 0>}
  {
   shapeType=Box
   extentX=25
   extentY=7.5
   extentZ=18
   shapeName=Front_Left
   center=<-7.9633 6.0611 -8.0609>
   rotation=<rot -2.535 2.489 45.5016>
  }
  {shapeType=Box, extentX=12.5, extentY=7.5, extentZ=22.5, shapeName=Back_Lower, center=<-9.8383 -4.3347 -4.8047>, rotation=<rot 0 -18.0573 0>}
  {shapeType=Box, extentX=8, extentY=6.5, extentZ=12.5, shapeName=Back_Top, center=<-1.3618 -6.5374 -7.28>, rotation=<rot 0 2.7518 0>}
 ]
 thigh_l [
  {shapeType=Capsule, length=40.875, radius=8.625, center=<-28.4068 0 0.4383>, rotation=<rot 89.0095 0 0>}
  {shapeType=Capsule, length=40.875, radius=8.625, center=<-28.3886 1.4073 0.6839>, rotation=<rot 89.0095 0 0>}
 ]
}
constraints [
 {bone1Name=dyn_tassel_frnt_r_parent, bone2Name=dyn_tassel_frnt_r_base, swing1LimitDegrees=10, swing2LimitDegrees=10, twistLimitDegrees=10}
 {bone1Name=dyn_tassel_frnt_r_01, bone2Name=dyn_tassel_frnt_r_parent, swing1LimitDegrees=10, swing2LimitDegrees=10, twistLimitDegrees=10}
 {bone1Name=dyn_tassel_frnt_r_02, bone2Name=dyn_tassel_frnt_r_01, swing1LimitDegrees=10, swing2LimitDegrees=10, twistLimitDegrees=10}
 {bone1Name=dyn_tassel_frnt_r_03, bone2Name=dyn_tassel_frnt_r_02, swing1LimitDegrees=10, swing2LimitDegrees=10, twistLimitDegrees=10}
 {bone1Name=dyn_tassel_frnt_l_parent, bone2Name=dyn_bag_base, swing1LimitDegrees=10, swing2LimitDegrees=10, twistLimitDegrees=10}
 {bone1Name=dyn_tassel_frnt_l_01, bone2Name=dyn_tassel_frnt_l_parent, swing1LimitDegrees=10, swing2LimitDegrees=10, twistLimitDegrees=10}
 {bone1Name=dyn_tassel_frnt_l_02, bone2Name=dyn_tassel_frnt_l_01, swing1LimitDegrees=10, swing2LimitDegrees=10, twistLimitDegrees=10}
 {bone1Name=dyn_tassel_frnt_l_03, bone2Name=dyn_tassel_frnt_l_02, swing1LimitDegrees=10, swing2LimitDegrees=10, twistLimitDegrees=10}
 {bone1Name=dyn_tassel_side_parent, bone2Name=dyn_tassel_side_base, swing1LimitDegrees=30, swing2LimitDegrees=30, twistLimitDegrees=50}
 {bone1Name=dyn_tassel_side_01, bone2Name=dyn_tassel_side_parent, swing1LimitDegrees=20, swing2LimitDegrees=20, twistLimitDegrees=50}
 {bone1Name=dyn_tassel_side_02, bone2Name=dyn_tassel_side_01, swing1LimitDegrees=20, swing2LimitDegrees=20, twistLimitDegrees=50}
 {bone1Name=dyn_tassel_side_03, bone2Name=dyn_tassel_side_02, swing1LimitDegrees=20, swing2LimitDegrees=20, twistLimitDegrees=20}
 {bone1Name=dyn_shirt_back_parent, bone2Name=dyn_shirt_back_base, swing1LimitDegrees=15, swing2LimitDegrees=60, twistLimitDegrees=5}
 {bone1Name=dyn_pouch_top_r_parent, bone2Name=dyn_pouch_top_r_base, swing1LimitDegrees=20, swing2LimitDegrees=20, twistLimitDegrees=3}
 {bone1Name=dyn_bag_parent, bone2Name=dyn_bag_base, swing1LimitDegrees=2, swing2LimitDegrees=10, twistLimitDegrees=1}
 {
  bone1Name=dyn_tassel_frnt_l_base
  bone2Name=dyn_bag_base
  swing1LimitDegrees=45
  swing1Motion=Free
  swing2LimitDegrees=45
  swing2Motion=Free
  twistLimitDegrees=45
  twistMotion=Free
 }
]
