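# Physics asset for the medium male Street Rat character body, covering secondary motion of its bag and back tassel.
# Capsules on pelvis through spine_03 form the torso, linked with free angular constraints up to a small box at the bag base.
# From that base, the bag capsule hangs on tight angular limits, while the tassel base attaches freely and leads a four-capsule chain with moderate swing and narrow twist limits.
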PhysicsAsset /BRCosmetics/Characters/Player/Male/Medium/Bodies/M_MED_Street_Rat/Meshes/M_MED_Street_Rat_Tassle_Physics
bodies {
 dyn_bag_base [{shapeType=Box, extentX=1.01, extentY=1.01, extentZ=1.01, rotation=<rot 0 -90 0>}]
 dyn_bag_parent [{
   shapeType=Capsule
   length=10.8263
   radius=5.9766
   shapeName=dyn_bag_parent_capsule
   center=<8.1665 1.5835 1.6421>
   rotation=<rot 84.282 173.1472 179.9999>
  }]
 dyn_tassel_back_01 [{shapeType=Capsule, length=0.0002, radius=1.5437, shapeName=dyn_tassel_back_01_capsule, center=<1.5 0 0>}]
 dyn_tassel_back_02 [{shapeType=Capsule, length=0.5733, radius=1.3991, shapeName=dyn_tassel_back_02_capsule, center=<1.5 0 0>, rotation=<rot 90 0 0>}]
 dyn_tassel_back_03 [{shapeType=Capsule, length=1.2216, radius=2.03, shapeName=dyn_tassel_back_03_capsule, center=<3 0 0>, rotation=<rot 90 0 0>}]
 dyn_tassel_back_base [{shapeType=Box, extentX=0.5, extentY=0.5, extentZ=0.5}]
 dyn_tassel_back_parent [{shapeType=Capsule, length=2.7298, radius=0.8291, shapeName=dyn_tassel_back_parent_capsule, center=<2 0 0>, rotation=<rot 90 0 0>}]
 pelvis [{shapeType=Capsule, length=5.1021, radius=13.8339, shapeName=pelvis_capsule, center=<0 -0.5846 0>}]
 spine_01 [{shapeType=Capsule, length=8.7573, radius=10.8654, shapeName=spine_01_capsule, center=<0 -0.8479 0>}]
 spine_02 [{shapeType=Capsule, length=4.1913, radius=11.1008, shapeName=spine_02_capsule, center=<0 0.3755 0>}]
 spine_03 [{shapeType=Capsule, length=3.0772, radius=11.8347, shapeName=spine_03_capsule, center=<0 0.7688 0>}]
}
constraints [
 {bone1Name=dyn_bag_parent, bone2Name=dyn_bag_base, swing1LimitDegrees=2, swing2LimitDegrees=10, twistLimitDegrees=1}
 {
  bone1Name=dyn_tassel_back_base
  bone2Name=dyn_bag_base
  swing1LimitDegrees=45
  swing1Motion=Free
  swing2LimitDegrees=45
  swing2Motion=Free
  twistLimitDegrees=45
  twistMotion=Free
 }
 {
  bone1Name=dyn_bag_base
  bone2Name=spine_03
  swing1LimitDegrees=45
  swing1Motion=Free
  swing2LimitDegrees=45
  swing2Motion=Free
  twistLimitDegrees=45
  twistMotion=Free
 }
 {
  bone1Name=spine_03
  bone2Name=spine_02
  swing1LimitDegrees=45
  swing1Motion=Free
  swing2LimitDegrees=45
  swing2Motion=Free
  twistLimitDegrees=45
  twistMotion=Free
 }
 {
  bone1Name=spine_02
  bone2Name=spine_01
  swing1LimitDegrees=45
  swing1Motion=Free
  swing2LimitDegrees=45
  swing2Motion=Free
  twistLimitDegrees=45
  twistMotion=Free
 }
 {
  bone1Name=spine_01
  bone2Name=pelvis
  swing1LimitDegrees=45
  swing1Motion=Free
  swing2LimitDegrees=45
  swing2Motion=Free
  twistLimitDegrees=45
  twistMotion=Free
 }
 {bone1Name=dyn_tassel_back_03, bone2Name=dyn_tassel_back_02, swing1LimitDegrees=20, swing2LimitDegrees=15, twistLimitDegrees=5}
 {bone1Name=dyn_tassel_back_01, bone2Name=dyn_tassel_back_parent, swing1LimitDegrees=20, swing2LimitDegrees=15, twistLimitDegrees=5}
 {bone1Name=dyn_tassel_back_02, bone2Name=dyn_tassel_back_01, swing1LimitDegrees=20, swing2LimitDegrees=15, twistLimitDegrees=5}
 {bone1Name=dyn_tassel_back_parent, bone2Name=dyn_tassel_back_base, swing1LimitDegrees=20, swing2LimitDegrees=15, twistLimitDegrees=5}
]
