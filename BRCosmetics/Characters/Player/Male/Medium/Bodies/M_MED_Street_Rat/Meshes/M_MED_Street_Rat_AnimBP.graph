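# Post-process style cosmetic AnimBP for the Street Rat body, adding secondary motion to straps, tassels, shoe tags, zipper and shirt.
# It caches the copied parent pose after AnimDynamics and bone constraints.
# Crouch and standing branches add strap rotations weighted by left-thigh angle and separate RigidBody setups, blend on bIsCrouching, then feed a final tassel simulation.
# The event graph computes thigh alphas and gravity override through a deprecated function.
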AnimBlueprint /BRCosmetics/Characters/Player/Male/Medium/Bodies/M_MED_Street_Rat/Meshes/M_MED_Street_Rat_AnimBP
target_skeleton=&/M_MED_Street_Rat_Skeleton
graphs [
 AnimGraph {nodes [
   n1 = "Transform (Modify) Bone - Bone: dyn_tassel_side_base"(ComponentPose=Local To Component(LocalPose=Use cached pose 'CrouchCache'()), Alpha=1, Translation="0,5,0", Rotation="0,0,-50", Scale="1,1,1") {
    rotationMode=BMM_Additive
    rotationSpace=BCS_BoneSpace
    translationMode=BMM_Additive
    translationSpace=BCS_BoneSpace
    alphaBoolBlend {blendOption=Linear}
    boneToModify {boneName=dyn_tassel_side_base}
    rotation {roll=-50}
    translation {y=5}
   }
   n2 = "Transform (Modify) Bone - Bone: dyn_strap_base"(ComponentPose=@n1, Alpha=Get Thigh_Rot_L() -> (Thigh_Rot_L:double), Translation="0,0,0", Rotation="-100,0,0", Scale="1,1,1") {
    rotationMode=BMM_Additive
    rotationSpace=BCS_BoneSpace
    alphaBoolBlend {blendOption=Linear}
    alphaScaleBiasClamp {bInterpResult=true, interpSpeedDecreasing=8, interpSpeedIncreasing=8}
    boneToModify {boneName=dyn_strap_base}
    rotation {pitch=-100}
   }
   n3 = RigidBody(ComponentPose=@n2) {
    bClampLinearTranslationLimitToRefPose=true
    bOverrideWorldGravity=true
    overridePhysicsAsset=&/M_MED_Street_Rat_CrouchingTassels_Physics
    simulationSpace=BaseBoneSpace
    alphaBoolBlend {blendOption=Linear}
    baseBoneRef {boneName=pelvis}
    componentAppliedLinearAccClamp=<6000 6000 6000>
    componentLinearAccScale=<0.25 0.25 0.25>
    componentLinearVelScale=<0.25 0.25 0.25>
    overrideWorldGravity {z=-980}
   }
   n4 = Component To Local(ComponentPose=@n3)
   n5 = Get Thigh_Rot_L()
   n6 = "Transform (Modify) Bone - Bone: dyn_strap_base"(
    ComponentPose=Local To Component(LocalPose=Use cached pose 'CrouchCache'())
    Alpha=@n5
    Translation="0,0,0"
    Rotation="-50,0,0"
    Scale="1,1,1"
   ) {
    rotationMode=BMM_Additive
    rotationSpace=BCS_BoneSpace
    alphaBoolBlend {blendOption=Linear}
    boneToModify {boneName=dyn_strap_base}
    rotation {pitch=-50}
   }
   n7 = "Transform (Modify) Bone - Bone: dyn_strap_02"(ComponentPose=@n6, Alpha=@n5, Translation="0,0,0", Rotation="-50,0,0", Scale="1,1,1") {
    rotationMode=BMM_Additive
    rotationSpace=BCS_BoneSpace
    alphaBoolBlend {blendOption=Linear}
    boneToModify {boneName=dyn_strap_02}
    rotation {pitch=-50}
   }
   n8 = RigidBody(ComponentPose=@n7) {
    bClampLinearTranslationLimitToRefPose=true
    bOverrideWorldGravity=true
    overridePhysicsAsset=&/M_MED_Street_Rat_FrontStrap_Physics
    simulationSpace=BaseBoneSpace
    alphaBoolBlend {blendOption=Linear}
    alphaScaleBiasClamp {bInterpResult=true, interpSpeedDecreasing=4, interpSpeedIncreasing=4}
    baseBoneRef {boneName=root}
    componentAppliedLinearAccClamp=<6000 6000 6000>
    componentLinearAccScale=<0.15 0.15 0.15>
    componentLinearVelScale=<0.15 0.15 0.15>
    overrideWorldGravity {z=-980}
   }
   n9 = RigidBody!(ComponentPose=@n8, Alpha, OverrideWorldGravity:Vector=Get GravityOverride(), ExternalForce) -> (Pose) {
    bClampLinearTranslationLimitToRefPose=true
    bOverrideWorldGravity=true
    overridePhysicsAsset=&/M_MED_Street_Rat_Physics
    simulationSpace=BaseBoneSpace
    alphaBoolBlend {blendOption=Linear}
    baseBoneRef {boneName=spine_02}
    componentAppliedLinearAccClamp=<6000 6000 6000>
    componentLinearAccScale=<0.25 0.25 0.25>
    componentLinearVelScale=<0.25 0.25 0.25>
    overrideWorldGravity {z=-980}
   }
   n10 = Anim Dynamics - Bone: dyn_shirt_back_parent(ComponentPose=@n9, Alpha=1, ExternalForce="0,0,0") {
    angularDampingOverride=0.9
    bGravityOverrideInSimSpace=true
    bOverrideAngularDamping=true
    bOverrideLinearDamping=true
    gravityScale=0.15
    linearDampingOverride=0.9
    simulationSpace=Actor
    alphaBoolBlend {blendOption=Linear}
    boundBone {boneName=dyn_shirt_back_parent}
    componentAppliedLinearAccClamp=<6000 6000 6000>
    componentLinearAccScale=<0.15 0.15 0.15>
    componentLinearVelScale=<0.15 0.15 0.15>
    gravityOverride=<0.9 0.9 0.9>
    physicsBodyDefinitions [{
      sphereCollisionRadius=0
      boundBone {boneName=dyn_shirt_back_parent}
      boxExtents=<10 8 2>
      constraintSetup {angularLimitsMax=<1 45 5>, angularLimitsMin=<-1 -10 -5>, angularTarget=<0 0 0>}
      localJointOffset=<3.5 0 0>
     }]
    relativeSpaceBone {boneName=root}
   }
   n11 = Component To Local(ComponentPose=@n10)
   n12 = Get bIsCrouching()
   n13 = RigidBody!(
    ComponentPose=Local To Component(LocalPose=Blend Poses by bool(BlendPose_0=@n4, BlendPose_1=@n11, BlendTime_0=0.01, BlendTime_1=0.75, bActiveValue=@n12) {blendType=Linear})
    Alpha
    ExternalForce
   ) -> (Pose) {
    bClampLinearTranslationLimitToRefPose=true
    bOverrideWorldGravity=true
    overridePhysicsAsset=&/M_MED_Street_Rat_Tassle_Physics
    simulationSpace=BaseBoneSpace
    alphaBoolBlend {blendOption=Linear}
    baseBoneRef {boneName=spine_02}
    componentAppliedLinearAccClamp=<6000 6000 6000>
    componentLinearAccScale=<0.25 0.25 0.25>
    componentLinearVelScale=<0.25 0.25 0.25>
    overrideWorldGravity {z=-980}
   }
   Output Pose(Result=Component To Local(ComponentPose=@n13)) {layerGroup=None}
   n14 = Anim Dynamics - Bone: dyn_shoetag_l_parent(ComponentPose=Local To Component(LocalPose=Copy Pose From Mesh() {bUseAttachedParent=true}), Alpha=1, ExternalForce="0,0,0") {
    angularDampingOverride=0.9
    bGravityOverrideInSimSpace=true
    bOverrideAngularDamping=true
    bOverrideLinearDamping=true
    gravityScale=0.15
    linearDampingOverride=0.9
    simulationSpace=Actor
    alphaBoolBlend {blendOption=Linear}
    boundBone {boneName=dyn_shoetag_l_parent}
    componentAppliedLinearAccClamp=<6000 6000 6000>
    componentLinearAccScale=<0.15 0.15 0.15>
    componentLinearVelScale=<0.15 0.15 0.15>
    gravityOverride=<0.9 0.9 0.9>
    physicsBodyDefinitions [{
      sphereCollisionRadius=0
      boundBone {boneName=dyn_shoetag_l_parent}
      boxExtents=<10 8 2>
      constraintSetup {angularLimitsMax=<10 40 5>, angularLimitsMin=<-10 -10 -5>, angularTarget=<0 0 0>}
      localJointOffset=<3.5 0 0>
     }]
    relativeSpaceBone {boneName=root}
   }
   n15 = Anim Dynamics - Bone: dyn_shoetag_r_parent(ComponentPose=@n14, Alpha=1, ExternalForce="0,0,0") {
    angularDampingOverride=0.9
    bGravityOverrideInSimSpace=true
    bOverrideAngularDamping=true
    bOverrideLinearDamping=true
    gravityScale=0.15
    linearDampingOverride=0.9
    simulationSpace=Actor
    alphaBoolBlend {blendOption=Linear}
    boundBone {boneName=dyn_shoetag_r_parent}
    componentAppliedLinearAccClamp=<6000 6000 6000>
    componentLinearAccScale=<0.15 0.15 0.15>
    componentLinearVelScale=<0.15 0.15 0.15>
    gravityOverride=<0.9 0.9 0.9>
    physicsBodyDefinitions [{
      sphereCollisionRadius=0
      boundBone {boneName=dyn_shoetag_r_parent}
      boxExtents=<10 8 2>
      constraintSetup {angularLimitsMax=<10 40 5>, angularLimitsMin=<-10 -10 -5>, angularTarget=<0 0 0>}
      localJointOffset=<3.5 0 0>
     }]
    relativeSpaceBone {boneName=root}
   }
   n16 = "Constraint - dyn_shirt_back_base (PPPPP)"!(
    ComponentPose=@n15
    Alpha=1
    ConstraintWeights_0=0.8702
    ConstraintWeights_1:float=0.0034
    ConstraintWeights_2:float=0.016
    ConstraintWeights_3:float=0.0191
    ConstraintWeights_4:float=0.0914
   ) -> (Pose) {
    alphaBoolBlend {blendOption=Linear}
    boneToModify {boneName=dyn_shirt_back_base}
    constraintSetup [
     {targetBone {boneName=pelvis}}
     {targetBone {boneName=spine_03}}
     {targetBone {boneName=thigh_twist_01_l}}
     {targetBone {boneName=thigh_twist_01_r}}
     {targetBone {boneName=spine_01}}
    ]
    constraintWeights [1,1,1,1,1]
   }
   n17 = "Constraint - dyn_zipper_r_base (PPP)"!(ComponentPose=@n16, Alpha=1, ConstraintWeights_0=0.1653, ConstraintWeights_1:float=0.0003, ConstraintWeights_2:float=0.8343) -> (Pose) {
    alphaBoolBlend {blendOption=Linear}
    boneToModify {boneName=dyn_zipper_r_base}
    constraintSetup [{targetBone {boneName=thigh_r}}, {targetBone {boneName=pelvis}}, {targetBone {boneName=thigh_twist_01_r}}]
    constraintWeights [1,1,1]
   }
   n18 = Anim Dynamics - Bone: dyn_zipper_r_parent(ComponentPose=@n17, Alpha=1, ExternalForce="0,0,0") {
    angularDampingOverride=0.99
    bGravityOverrideInSimSpace=true
    bOverrideLinearDamping=true
    gravityScale=0.75
    linearDampingOverride=0.999
    alphaBoolBlend {blendOption=Linear}
    boundBone {boneName=dyn_zipper_r_parent}
    componentAppliedLinearAccClamp=<6000 6000 6000>
    componentLinearAccScale=<0.25 0.25 0.15>
    componentLinearVelScale=<0.25 0.25 0.15>
    physicsBodyDefinitions [{
      sphereCollisionRadius=0
      boundBone {boneName=dyn_zipper_r_parent}
      boxExtents=<5 1.5 4>
      constraintSetup {angularLimitsMax=<0 5 50>, angularLimitsMin=<0 -5 -10>, angularTarget=<0 0 0>}
      localJointOffset=<2 0 0>
     }]
    relativeSpaceBone {boneName=spine_05}
   }
   n19 = "Constraint - dyn_tassel_side_base (P)"!(ComponentPose=@n18, Alpha=1, ConstraintWeights_0=1, ConstraintWeights_1:float=0) -> (Pose) {
    alphaBoolBlend {blendOption=Linear}
    boneToModify {boneName=dyn_tassel_side_base}
    constraintSetup [{targetBone {boneName=dyn_bag_parent}}]
    constraintWeights [1,0]
   }
   n20 = "Constraint - dyn_tassel_frnt_r_base (P)"!(ComponentPose=@n19, Alpha=1, ConstraintWeights_0=1, ConstraintWeights_1:float=0) -> (Pose) {
    alphaBoolBlend {blendOption=Linear}
    boneToModify {boneName=dyn_tassel_frnt_r_base}
    constraintSetup [{targetBone {boneName=dyn_bag_parent}}]
    constraintWeights [1,0]
   }
   n21 = "Constraint - dyn_tassel_back_base (P)"!(ComponentPose=@n20, Alpha=1, ConstraintWeights_0=1, ConstraintWeights_1:float=0) -> (Pose) {
    alphaBoolBlend {blendOption=Linear}
    boneToModify {boneName=dyn_tassel_back_base}
    constraintSetup [{targetBone {boneName=dyn_bag_parent}}]
    constraintWeights [1,0]
   }
   n22 = "Constraint - dyn_strap_base (PPP)"!(ComponentPose=@n21, Alpha=1, ConstraintWeights_0=0.6236, ConstraintWeights_1:float=0.2926, ConstraintWeights_2:float=0.0838) -> (Pose) {
    alphaBoolBlend {blendOption=Linear}
    boneToModify {boneName=dyn_strap_base}
    constraintSetup [{targetBone {boneName=pelvis}}, {targetBone {boneName=thigh_twist_01_l}}, {targetBone {boneName=spine_01}}]
    constraintWeights [1,1,1]
   }
   n23 = "Constraint - dyn_pouch_top_r_base (PPP)"!(ComponentPose=@n22, Alpha=1, ConstraintWeights_0=0.8353, ConstraintWeights_1:float=0.0081, ConstraintWeights_2:float=0.1566) -> (Pose) {
    alphaBoolBlend {blendOption=Linear}
    boneToModify {boneName=dyn_pouch_top_r_base}
    constraintSetup [{targetBone {boneName=pelvis}}, {targetBone {boneName=thigh_twist_01_r}}, {targetBone {boneName=spine_01}}]
    constraintWeights [1,1,1]
   }
   Save cached pose 'CrouchCache'(Pose=Component To Local(ComponentPose=@n23)) {cacheName=CrouchCache}
  ]}
 EventGraph {nodes [
   Event BlueprintUpdateAnimation()
   > n1 = Sequence()
   n2 = GetOwnerSkeletalMeshComponent()
   n3 = BreakRotator(InRot=BreakTransform(InTransform=GetDeltaTransformFromRefPose(self=@n2, BoneName=thigh_l, BaseName=pelvis)).Rotation)
   > Set Thigh_Rotation_L(Thigh_Rotation_L:double=MapRangeClamped(Value=@n3.Yaw, InRangeA=-50, InRangeB=-90)) -> (Output_Get:double)
   > Set Thigh_Rot_L(Thigh_Rot_L:double=MapRangeClamped(Value=@n3.Yaw, InRangeA=-15, InRangeB=-110)) -> (Output_Get:double)
   Set Crouch_Alpha(execute=@n1.then_1, Crouch_Alpha:double=SelectFloat(A=1, B=0, bPickA=Get bIsCrouching())) -> (Output_Get:double)
   > n4 = Gravity Override State Function DEPRECATED(CharacterPartAnimInstance=Self-Reference(), GameTimeInSeconds=GetGameTimeInSeconds(), GravityOverrideParams=Get Gravity Override Params())
   > Set GravityOverride(GravityOverride=@n4)
  ]}
]
variables [
 Crouch_Alpha {type=number}
 Gravity Override Params {type=GravityOverrideParamsStruct}
 GravityOverride {type=Vector}
 Thigh_Rot_L {type=number}
 Thigh_Rotation_L {type=number}
]
