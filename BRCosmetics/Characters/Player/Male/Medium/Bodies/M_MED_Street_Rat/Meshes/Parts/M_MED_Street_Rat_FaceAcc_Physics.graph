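# Physics asset for the Street Rat male medium face accessory, covering a hood top and left and right hood strings.
# The hood top box is tightly limited against the head, while each string chains from a small base box through four capsules with wider swing and narrow twist.
# A head sphere and two spine_05 capsules supply body shapes.
# Right-string capsules are fatter and mostly near zero length, so they act almost like spheres, unlike the left.
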PhysicsAsset /BRCosmetics/Characters/Player/Male/Medium/Bodies/M_MED_Street_Rat/Meshes/Parts/M_MED_Street_Rat_FaceAcc_Physics
bodies {
 dyn_hood_top [{shapeType=Box, extentX=3.76, extentY=4.01, extentZ=1, center=<1.8861 0 0.0002>}]
 dyn_string_l_01 [{
   shapeType=Capsule
   length=1.8884
   radius=1.1596
   shapeName=dyn_string_l_01_capsule
   center=<2 0 0>
   rotation=<rot 84.2023 -179.9999 -179.9999>
  }]
 dyn_string_l_02 [{
   shapeType=Capsule
   length=1.5289
   radius=1.38
   shapeName=dyn_string_l_02_capsule
   center=<1.9406 0 0>
   rotation=<rot 83.2522 179.9998 179.9998>
  }]
 dyn_string_l_03 [{shapeType=Capsule, length=0.3193, radius=1.2725, shapeName=dyn_string_l_03_capsule, center=<1.2601 0 0>, rotation=<rot 90 0 0>}]
 dyn_string_l_base [{shapeType=Box, extentX=0.5, extentY=0.5, extentZ=0.5}]
 dyn_string_l_parent [{shapeType=Capsule, length=2.2793, radius=1.0479, shapeName=dyn_string_l_parent_capsule, center=<2.1252 0 0>, rotation=<rot 90 0 0>}]
 dyn_string_r_01 [{shapeType=Capsule, length=0.0002, radius=2.8655, shapeName=dyn_string_r_01_capsule, center=<2.0797 0 0>, rotation=<rot 90 0 0>}]
 dyn_string_r_02 [{shapeType=Capsule, length=0.0002, radius=3.0902, shapeName=dyn_string_r_02_capsule, center=<1.9858 0 0>, rotation=<rot 90 0 0>}]
 dyn_string_r_03 [{shapeType=Capsule, length=0.4617, radius=2.523, shapeName=dyn_string_r_03_capsule, center=<1.2803 0 0>, rotation=<rot 90 0 0>}]
 dyn_string_r_base [{shapeType=Box, extentX=0.5, extentY=0.5, extentZ=0.5, rotation=<rot 0.0149 1.0908 0.3623>}]
 dyn_string_r_parent [{shapeType=Capsule, length=0.0002, radius=2.7709, shapeName=dyn_string_r_parent_capsule, center=<1.9577 0 0>, rotation=<rot 90 0 0>}]
 head [{shapeType=Sphere, radius=8, center=<5.615 4.6544 0>}]
 spine_05 [
  {shapeType=Capsule, length=11, radius=10, shapeName=spine_05_capsule, center=<-1 0.9862 0>}
  {shapeType=Capsule, length=10, radius=10, shapeName=spine_05_capsule, center=<-6.0349 4.19 0>}
 ]
}
constraints [
 {bone1Name=dyn_hood_top, bone2Name=head, swing1LimitDegrees=10, swing2LimitDegrees=20, twistLimitDegrees=2}
 {bone1Name=dyn_string_l_03, bone2Name=dyn_string_l_02, swing1LimitDegrees=35, swing2LimitDegrees=35, twistLimitDegrees=5}
 {bone1Name=dyn_string_l_parent, bone2Name=dyn_string_l_base, swing1LimitDegrees=35, swing2LimitDegrees=35, twistLimitDegrees=5}
 {bone1Name=dyn_string_l_01, bone2Name=dyn_string_l_parent, swing1LimitDegrees=35, swing2LimitDegrees=35, twistLimitDegrees=5}
 {bone1Name=dyn_string_l_02, bone2Name=dyn_string_l_01, swing1LimitDegrees=35, swing2LimitDegrees=35, twistLimitDegrees=5}
 {bone1Name=dyn_string_r_03, bone2Name=dyn_string_r_02, swing1LimitDegrees=35, swing2LimitDegrees=35, twistLimitDegrees=5}
 {bone1Name=dyn_string_r_parent, bone2Name=dyn_string_r_base, swing1LimitDegrees=35, swing2LimitDegrees=35, twistLimitDegrees=5}
 {bone1Name=dyn_string_r_01, bone2Name=dyn_string_r_parent, swing1LimitDegrees=35, swing2LimitDegrees=35, twistLimitDegrees=5}
 {bone1Name=dyn_string_r_02, bone2Name=dyn_string_r_01, swing1LimitDegrees=35, swing2LimitDegrees=35, twistLimitDegrees=5}
]
