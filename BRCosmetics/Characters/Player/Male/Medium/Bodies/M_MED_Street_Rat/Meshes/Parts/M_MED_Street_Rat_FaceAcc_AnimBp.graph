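# Animation Blueprint for the Street Rat face accessory mesh, following the attached parent mesh and adding secondary motion.
# The copied parent pose feeds weighted multi-bone constraints that place the zipper and left/right string bases between head, neck, clavicle and upper spine.
# Anim Dynamics swings the zipper after its constraint, and a final RigidBody node simulates the accessory physics asset in spine_02 space.
# The event graph holds no logic.
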AnimBlueprint /BRCosmetics/Characters/Player/Male/Medium/Bodies/M_MED_Street_Rat/Meshes/Parts/M_MED_Street_Rat_FaceAcc_AnimBp
target_skeleton=&/M_MED_Street_Rat_FaceAcc_Skeleton
graphs [
 AnimGraph {nodes [
   n1 = "Constraint - dyn_zipper_base (PPPPPPP)"!(
    ComponentPose=Local To Component(LocalPose=Copy Pose From Mesh() {bUseAttachedParent=true})
    Alpha=1
    ConstraintWeights_0=0.0711
    ConstraintWeights_1:float=0.0013
    ConstraintWeights_2:float=0.0052
    ConstraintWeights_3:float=0.6271
    ConstraintWeights_4:float=0.0011
    ConstraintWeights_5:float=0.1035
    ConstraintWeights_6:float=0.1907
   ) -> (Pose) {
    alphaBoolBlend {blendOption=Linear}
    boneToModify {boneName=dyn_zipper_base}
    constraintSetup [
     {targetBone {boneName=head}}
     {targetBone {boneName=clavicle_l}}
     {targetBone {boneName=spine_04}}
     {targetBone {boneName=spine_05}}
     {targetBone {boneName=clavicle_r}}
     {targetBone {boneName=neck_02}}
     {targetBone {boneName=neck_01}}
    ]
    constraintWeights [1,1,1,1,1,1,1]
   }
   n2 = Anim Dynamics - Bone: dyn_zipper_parent(ComponentPose=@n1, Alpha=1, ExternalForce="0,0,0") {
    angularDampingOverride=0.99
    bGravityOverrideInSimSpace=true
    bOverrideLinearDamping=true
    gravityScale=0.75
    linearDampingOverride=0.999
    alphaBoolBlend {blendOption=Linear}
    boundBone {boneName=dyn_zipper_parent}
    componentAppliedLinearAccClamp=<6000 6000 6000>
    componentLinearAccScale=<0.25 0.25 0.15>
    componentLinearVelScale=<0.25 0.25 0.15>
    physicsBodyDefinitions [{
      sphereCollisionRadius=0
      boundBone {boneName=dyn_zipper_parent}
      boxExtents=<5 1.5 4>
      constraintSetup {angularLimitsMax=<0 5 50>, angularLimitsMin=<0 -5 -10>, angularTarget=<0 0 0>}
      localJointOffset=<2 0 0>
     }]
    relativeSpaceBone {boneName=root}
   }
   n3 = "Constraint - dyn_string_r_base (PPPPPP)"!(
    ComponentPose=@n2
    Alpha=1
    ConstraintWeights_0=0.1007
    ConstraintWeights_1:float=0.0022
    ConstraintWeights_2:float=0.5629
    ConstraintWeights_3:float=0.0057
    ConstraintWeights_4:float=0.1314
    ConstraintWeights_5:float=0.1972
   ) -> (Pose) {
    alphaBoolBlend {blendOption=Linear}
    boneToModify {boneName=dyn_string_r_base}
    constraintSetup [
     {targetBone {boneName=head}}
     {targetBone {boneName=spine_04}}
     {targetBone {boneName=spine_05}}
     {targetBone {boneName=clavicle_r}}
     {targetBone {boneName=neck_02}}
     {targetBone {boneName=neck_01}}
    ]
    constraintWeights [1,1,1,1,1,1]
   }
   n4 = "Constraint - dyn_string_l_base (PPPPPP)"!(
    ComponentPose=@n3
    Alpha=1
    ConstraintWeights_0=0.0786
    ConstraintWeights_1:float=0.0034
    ConstraintWeights_2:float=0.0019
    ConstraintWeights_3:float=0.5551
    ConstraintWeights_4:float=0.1455
    ConstraintWeights_5:float=0.2155
   ) -> (Pose) {
    alphaBoolBlend {blendOption=Linear}
    boneToModify {boneName=dyn_string_l_base}
    constraintSetup [
     {targetBone {boneName=head}}
     {targetBone {boneName=clavicle_l}}
     {targetBone {boneName=spine_04}}
     {targetBone {boneName=spine_05}}
     {targetBone {boneName=neck_02}}
     {targetBone {boneName=neck_01}}
    ]
    constraintWeights [1,1,1,1,1,1]
   }
   n5 = RigidBody(ComponentPose=@n4) {
    bClampLinearTranslationLimitToRefPose=true
    bOverrideWorldGravity=true
    overridePhysicsAsset=&/M_MED_Street_Rat_FaceAcc_Physics
    simulationSpace=BaseBoneSpace
    alphaBoolBlend {blendOption=Linear}
    baseBoneRef {boneName=spine_02}
    componentAppliedLinearAccClamp=<6000 6000 6000>
    componentLinearAccScale=<0.25 0.25 0.25>
    componentLinearVelScale=<0.25 0.25 0.25>
    overrideWorldGravity {z=-980}
   }
   Output Pose(Result=Component To Local(ComponentPose=@n5)) {layerGroup=None}
  ]}
 EventGraph {nodes [
   TryGetPawnOwner()
   Event BlueprintUpdateAnimation()
  ]}
]
variables []
